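# Niagara system for the MontyKeyboard emote that adds rainbow glow and lights to the keyboard's keys.
# Airglow (faint sprites) and Lights (light renderer) each burst 13 particles, one per bone of the User.SK_Keyboard skeletal mesh.
# Bone index samples a color curve for each key's hue.
# Each update, particles snap to their bone, a local PositionAffectsScale module drives size or light radius from position, and an orientation-relative offset shifts them.
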NiagaraSystem /BRCosmetics/Effects/Fort_Effects/Emotes/MontyKeyboard/NS_MontyKeyboard_Emote_Keys
dependencies {
 usedDataInterfaces [&NiagaraDataInterfaceColorCurve]
 usedDynamicInputs [&Multiply_VectorByFloat, &UniformAOrBFloat]
 usedModules [
  &SystemState
  &self:SystemSpawnScript.SystemScriptSource.SystemScriptGraph.NiagaraNodeAssignment_2.SetVariables_#1
 ]
}
emitters {
 Airglow {
  emitterUpdateScript [
   &EmitterState(Life Cycle Mode=::NewEnumerator1, Loop Duration=7, UseLoopDelay=false)
   &SpawnBurst_Instantaneous(Spawn Count=13, Use Spawn Probability=false)
  ]
  particleSpawnScript [
   /Niagara/Modules/Spawn/Initialization/InitializeParticle(
    Color=<rgba 0 0 0 0>
    Lifetime=7
    Mesh Scale=&VectorFromFloat(Value=&MakeFloatFromVector(Vector=<1 1 1>, Channel=::NewEnumerator0))
    Module.Write Color=false
    Module.Write Lifetime=true
    Module.Write Mass=true
    Module.Write Material Random=true
    Module.Write Position=false
    Module.Write Scale=true
    Module.Write SpriteRotation=false
    Module.Write SpriteSize=false
    Position=&SimulationPosition()
   )
   /Game/Effects/Niagara/Modules/Groups/FN_SetSpawnGroup_v1/FN_SetSpawnGroup(Emitter Bone Count=13)
   &Color(Color=&MakeLinearColorFromVectorAndFloat(
     "Vector (RGB)"=&MakeVectorFromLinearColorRGB(Color=&ColorFromCurve(
       ColorCurve=DataInterface() {propertyValues {
         AlphaCurve {keys []}
         BlueCurve {keys [
           <at 1 0 RCIM_Cubic>
           <at 2 0>
           <at 3 0>
           <at 4 0.0052>
           <at 5 0.0044>
           <at 6 0.3467>
           {arriveTangent=0.2748, interpMode=RCIM_Cubic, leaveTangent=0.2748, time=7, value=0.8714}
           <at 8 0.8963>
           <at 9 0.8879>
           <at 10 0.9047>
           <at 11 0.956>
           <at 12 0.9047>
           <at 13 0.7913>
          ]}
         GreenCurve {keys [
           <at 1 0 RCIM_Cubic>
           <at 2 0.1022>
           <at 3 0.7157>
           <at 4 0.9911>
           <at 5 1>
           <at 6 0.9647>
           {arriveTangent=-0.2126, interpMode=RCIM_Cubic, leaveTangent=-0.2126, time=7, value=0.7913}
           <at 8 0.5395>
           <at 9 0.3278>
           <at 10 0.1529>
           <at 11 0.2831>
           <at 12 0.1221>
           <at 13 0.0044>
          ]}
         RedCurve {keys [
           <at 1 1 RCIM_Cubic>
           <at 2 1>
           <at 3 1>
           <at 4 1>
           <at 5 0.7682>
           <at 6 0.1413>
           {arriveTangent=-0.0696, interpMode=RCIM_Cubic, leaveTangent=-0.0696, time=7, value=0.0009}
           <at 8 0.0021>
           <at 9 0.0212>
           <at 10 0.227>
           <at 11 1>
           <at 12 1>
           <at 13 1>
          ]}
        }}
       CurveIndex=&MakeFloatFromInt(Int=/Niagara/DynamicInputs/Meshes/SpecificBone(Mesh:NiagaraDataInterfaceSkeletalMesh=$User.SK_Keyboard, SpecificBoneIndex:NiagaraInt32=$Particles.SpawnGroup))
      ))
     "Float (Alpha)"=1
    ))
  ]
  particleUpdateScript [
   &ParticleState()
   &CameraOffset(Camera Offset Amount=15) {module=CameraOffset001}
   &SampleSkeletalMeshSkeleton(Mesh:NiagaraDataInterfaceSkeletalMesh=$User.SK_Keyboard, Specific Bone Index:NiagaraInt32=$Particles.SpawnGroup)
   /Niagara/Modules/Spawn/Location/SkeletalMeshSkeletonLocation(Module.Inherit Velocity=false, Module.Use Sampled Bone as Ribbon ID=true)
   &self:PositionAffectsScale(
    InputMax=1.7
    InputMin=-1.5
    LightRadius-Multiplier=100
    MeshScale-Multiplier:Vector3f=<1 1 1>
    OutputMax=1
    SpriteScale-Multiplier:Vector2f=<50 50>
   ) {module=PositionAffectsScale}
   /Game/Effects/Niagara/Modules/Position/FN_OffsetPositionByOrientationAxis(Offset Axis:Vector3f=<0 -8 1>) {module=OffsetPositionByMeshOrientation}
   &ScaleColor(Scale Alpha=0.15, Scale RGB=<0.65 0.65 0.65>, ScaleA=true, ScaleRGB=true)
  ]
  rendererClasses [&NiagaraSpriteRendererProperties]
  renderers [{rendererClass=&NiagaraSpriteRendererProperties}]
 }
 Lights {
  emitterUpdateScript [
   &EmitterState(Life Cycle Mode=::NewEnumerator1, Loop Duration=7, UseLoopDelay=false)
   &SpawnBurst_Instantaneous(Spawn Count=13, Use Spawn Probability=false)
  ]
  particleSpawnScript [
   &InitializeParticle(
    Color=<rgba 0 0 0 0>
    Lifetime=7
    Mesh Scale=&VectorFromFloat(Value=&MakeFloatFromVector(Vector=<1 1 1>, Channel=::NewEnumerator0))
    Module.Write Color=false
    Module.Write Lifetime=true
    Module.Write Mass=true
    Module.Write Material Random=true
    Module.Write Position=false
    Module.Write Scale=true
    Module.Write SpriteRotation=false
    Module.Write SpriteSize=false
    Position=&SimulationPosition()
   )
   &FN_SetSpawnGroup(Emitter Bone Count=13)
   &Color(Color=&MakeLinearColorFromVectorAndFloat(
     "Vector (RGB)"=&MakeVectorFromLinearColorRGB(Color=&ColorFromCurve(
       ColorCurve=DataInterface() {propertyValues {
         AlphaCurve {keys []}
         BlueCurve {keys [
           <at 1 0 RCIM_Cubic>
           <at 2 0>
           <at 3 0>
           <at 4 0.0052>
           <at 5 0.0044>
           <at 6 0.3467>
           {arriveTangent=0.2748, interpMode=RCIM_Cubic, leaveTangent=0.2748, time=7, value=0.8714}
           <at 8 0.8963>
           <at 9 0.8879>
           <at 10 0.9047>
           <at 11 0.956>
           <at 12 0.9047>
           <at 13 0.7913>
          ]}
         GreenCurve {keys [
           <at 1 0 RCIM_Cubic>
           <at 2 0.1022>
           <at 3 0.7157>
           <at 4 0.9911>
           <at 5 1>
           <at 6 0.9647>
           {arriveTangent=-0.2126, interpMode=RCIM_Cubic, leaveTangent=-0.2126, time=7, value=0.7913}
           <at 8 0.5395>
           <at 9 0.3278>
           <at 10 0.1529>
           <at 11 0.2831>
           <at 12 0.1221>
           <at 13 0.0044>
          ]}
         RedCurve {keys [
           <at 1 1 RCIM_Cubic>
           <at 2 1>
           <at 3 1>
           <at 4 1>
           <at 5 0.7682>
           <at 6 0.1413>
           {arriveTangent=-0.0696, interpMode=RCIM_Cubic, leaveTangent=-0.0696, time=7, value=0.0009}
           <at 8 0.0021>
           <at 9 0.0212>
           <at 10 0.227>
           <at 11 1>
           <at 12 1>
           <at 13 1>
          ]}
        }}
       CurveIndex=&MakeFloatFromInt(Int=&SpecificBone(Mesh:NiagaraDataInterfaceSkeletalMesh=$User.SK_Keyboard, SpecificBoneIndex:NiagaraInt32=$Particles.SpawnGroup))
      ))
     "Float (Alpha)"=1
    ))
  ]
  particleUpdateScript [
   &ParticleState()
   &CameraOffset(Camera Offset Amount=15) {module=CameraOffset001}
   &SampleSkeletalMeshSkeleton(Mesh:NiagaraDataInterfaceSkeletalMesh=$User.SK_Keyboard, Specific Bone Index:NiagaraInt32=$Particles.SpawnGroup)
   &SkeletalMeshSkeletonLocation(Module.Inherit Velocity=false, Module.Use Sampled Bone as Ribbon ID=true)
   &FN_Light_v2(Light Exponent=2, Light Radius=1)
   &self:PositionAffectsScale(InputMax=1.7, InputMin=-1.5, LightRadius-Multiplier=100, OutputMax=1, SpriteScale-Multiplier:Vector2f=<1 1>) {module=PositionAffectsScale}
   &FN_OffsetPositionByOrientationAxis(Offset Axis:Vector3f=<0 -20 5>) {module=OffsetPositionByMeshOrientation}
   /SetVariables(Particles.Light Color:LinearColor=&MultiplyLinearColor(Linear Color=$Particles.Color, Scale Factor=<0.8 0.8 0.8 1>)) {isSetParameters=true}
  ]
  rendererClasses [&NiagaraLightRendererProperties]
  renderers [{rendererClass=&NiagaraLightRendererProperties, rendererIndex=0}]
 }
}
summary {
 systemName=NS_MontyKeyboard_Emote_Keys
 emitters [
  {emitterName=Airglow, rendererClasses [&NiagaraSpriteRendererProperties]}
  {emitterName=Lights, rendererClasses [&NiagaraLightRendererProperties]}
 ]
 userVariables [
  User.Initial Mesh Scale {type=Vector3f, description="", defaultValue {struct=Vector3f, value=<0 0 0>}}
  User.SK_Keyboard {
   type=NiagaraDataInterfaceSkeletalMesh
   defaultValue {
    struct=NiagaraExt_VariableValue_DataInterface
    value {dataInterface=&self:NiagaraDataInterfaceSkeletalMesh_0}
   }
  }
 ]
}
